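# Conversation database for the TurnVortex NPC, a HUNTR/X-themed character that greets players and routes them into further conversations.
# Its entry point leads to one speech node with a default line plus contextual greetings gated on first-ever conversation, match conversation count, or player outfit.
# First-meeting and outfit-reaction lines take their priority from named curve rows.
# The speech node then branches through link nodes to NPC library quests, basic services and TurnVortex-specific dialogue.
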ConversationDatabase /NPCLibrary/NPCs/TurnVortex/NPCLibrary_Conversation_TurnVortex
entry_points [{tag=Conversation.Entry.TurnVortex, destinations [#1]}]
nodes [
 ConversationEntryPointNode_7 {
  id=#1
  class=ConversationEntryPointNode
  outputs [#2]
  properties {entryTag {tagName=Conversation.Entry.TurnVortex}}
 }
 Speech_C_4 {
  id=#2
  class=Speech_C
  outputs [#3, #4, #5]
  properties {
   nodeName=""
   generalConfig {
    defaultMessage=Forget vacation. HUNTR/X is ready.
    contextualMessages [
     {
      message=Forget vacation. HUNTR/X is ready.
      contextRequirements [{requirement {class=FortControllerRequirement_IsFirstEverConversationWithNPC}}]
      priority {curve {rowName=Tandem.Speech.Priority.FirstEverInteraction}}
     }
     {
      message="If we can take down the Saja Boys, we can take down anything."
      contextRequirements [{requirement {class=FortControllerRequirement_CheckNPCMatchConversationCount, properties {nPCUniqueIdTag {tagName=AISpawnerData.Type.Tandem.TurnVortex}}}}]
      priority {curve {rowName=None}}
     }
     {
      message="Happy fans, happy Honmoon!"
      contextRequirements [{requirement {class=FortControllerRequirement_CheckNPCMatchConversationCount, properties {nPCUniqueIdTag {tagName=AISpawnerData.Type.Tandem.TurnVortex}}}}]
      priority {curve {rowName=None}}
     }
     {
      message=Thank you for being a fan.
      contextRequirements [{requirement {class=FortControllerRequirement_CheckNPCMatchConversationCount, properties {nPCUniqueIdTag {tagName=AISpawnerData.Type.Tandem.TurnVortex}}}}]
      priority {curve {rowName=None}}
     }
     {
      message="This Island seems like your kinda place, no?"
      contextRequirements [{requirement {
         class=FortControllerRequirement_HasCID
         properties {allowedCIds [/BRCosmetics/Athena/Items/Cosmetics/Characters/Character_QuailWink, /BRCosmetics/Athena/Items/Cosmetics/Characters/Character_LowerDrop]}
        }}]
      priority {curve {rowName=Tandem.Speech.Priority.OutfitReaction}}
     }
     {
      message="I watched all of your 2-second turtle videos. Yes, they were cute."
      contextRequirements [{requirement {
         class=FortControllerRequirement_HasCID
         properties {allowedCIds [/BRCosmetics/Athena/Items/Cosmetics/Characters/Character_EvokeFind, /BRCosmetics/Athena/Items/Cosmetics/Characters/Character_HybridBoot]}
        }}]
      priority {curve {rowName=Tandem.Speech.Priority.OutfitReaction}}
     }
     {
      message=HUNTR/X will wipe the floor with the Saja Boys.
      contextRequirements [{requirement {class=FortControllerRequirement_HasCID, properties {allowedCIds [/BRCosmetics/Athena/Items/Cosmetics/Characters/Character_BountyPond]}}}]
      priority {curve {rowName=Tandem.Speech.Priority.OutfitReaction}}
     }
    ]
   }
  }
 }
 ConversationLinkNode_2 {
  id=#5
  class=ConversationLinkNode
  properties {remoteEntryTag {tagName=Conversation.Entry.TurnVortex.Dialogue}}
 }
 ConversationLinkNode_0 {id=#4, class=ConversationLinkNode, properties {remoteEntryTag {tagName=Conversation.Entry.BasicServices}}}
 ConversationLinkNode_1 {id=#3, class=ConversationLinkNode, properties {remoteEntryTag {tagName=Conversation.Entry.NPCLibrary.Quests}}}
]
