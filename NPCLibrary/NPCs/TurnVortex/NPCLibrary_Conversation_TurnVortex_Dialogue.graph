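# Conversation database for the TurnVortex NPC's "Ask about..." dialogue, giving in-character lore as a HUNTR/X member.
# The entry tag opens a "How can I help?" speech hub offering the Honmoon, Authenticity and Saja Boys topics, plus a Back node.
# Each topic is a linear chain of Speech nodes, and all three converge on one shared ReturnToStart node labeled BACK.
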ConversationDatabase /NPCLibrary/NPCs/TurnVortex/NPCLibrary_Conversation_TurnVortex_Dialogue
entry_points [{tag=Conversation.Entry.TurnVortex.Dialogue, destinations [#1]}]
nodes [
 ConversationEntryPointNode_0 {
  id=#1
  class=ConversationEntryPointNode
  outputs [#2]
  properties {entryTag {tagName=Conversation.Entry.TurnVortex.Dialogue}}
 }
 Speech_C_0 {
  id=#2
  class=Speech_C
  outputs [#3, #4, #5, #6]
  properties {nodeName="", softTaskIcon=&T_UI_NPC_question, generalConfig {defaultMessage=How can I help?}}
  sub_nodes [ConversationChoiceNode_5 {class=ConversationChoiceNode, properties {defaultChoiceDisplayText=Ask about...}}]
 }
 Speech_C_3 {
  id=#6
  class=Speech_C
  outputs [#7]
  properties {
   nodeName=""
   softTaskIcon=&T_UI_NPC_question
   generalConfig {defaultMessage="The Saja Boys are our terrible, demonic rivals."}
  }
  sub_nodes [ConversationChoiceNode_1 {class=ConversationChoiceNode, properties {defaultChoiceDisplayText=The Saja Boys}}]
 }
 Speech_C_6 {
  id=#7
  class=Speech_C
  outputs [#8]
  properties {
   nodeName=""
   softTaskIcon=&T_UI_NPC_question
   generalConfig {defaultMessage="I have... complicated... feelings about their frontman, Jinu."}
  }
 }
 Speech_C_8 {
  id=#8
  class=Speech_C
  outputs [#9]
  properties {
   nodeName=""
   softTaskIcon=&T_UI_NPC_question
   generalConfig {defaultMessage="But in a faceoff, HUNTR/X will always win the day."}
  }
 }
 ReturnToStart_C_23 {
  id=#9
  class=ReturnToStart_C
  properties {softTaskIcon=&T_UI_NPC_Back}
  sub_nodes [ConversationChoiceNode_32 {class=ConversationChoiceNode, properties {defaultChoiceDisplayText=BACK}}]
 }
 Speech_C_2 {
  id=#5
  class=Speech_C
  outputs [#10]
  properties {
   nodeName=""
   softTaskIcon=&T_UI_NPC_question
   generalConfig {defaultMessage=Being authentically yourself is your greatest strength.}
  }
  sub_nodes [ConversationChoiceNode_2 {class=ConversationChoiceNode, properties {defaultChoiceDisplayText=Authenticity}}]
 }
 Speech_C_4 {
  id=#10
  class=Speech_C
  outputs [#11]
  properties {
   nodeName=""
   softTaskIcon=&T_UI_NPC_question
   generalConfig {defaultMessage=The fans want the real you. Your friends love the real you. And demons...}
  }
 }
 Speech_C_5 {
  id=#11
  class=Speech_C
  outputs [#9]
  properties {
   nodeName=""
   softTaskIcon=&T_UI_NPC_question
   generalConfig {defaultMessage="...demons don't usually care either way. But my point still stands! Be you!"}
  }
 }
 Back_C_0 {
  id=#4
  class=Back_C
  sub_nodes [ConversationChoiceNode_7 {class=ConversationChoiceNode, properties {defaultChoiceDisplayText=BACK}}]
 }
 Speech_C_7 {
  id=#3
  class=Speech_C
  outputs [#12]
  properties {
   nodeName=""
   softTaskIcon=&T_UI_NPC_question
   generalConfig {defaultMessage=The Honmoon is a magical veil that protects our world from Gwi-Ma's demons.}
  }
  sub_nodes [ConversationChoiceNode_0 {class=ConversationChoiceNode, properties {defaultChoiceDisplayText=The Honmoon}}]
 }
 Speech_C_15 {
  id=#12
  class=Speech_C
  outputs [#13]
  properties {
   nodeName=""
   softTaskIcon=&T_UI_NPC_question
   generalConfig {defaultMessage=It's also the source of our energy and power.}
  }
 }
 Speech_C_14 {
  id=#13
  class=Speech_C
  outputs [#14]
  properties {
   nodeName=""
   softTaskIcon=&T_UI_NPC_question
   generalConfig {defaultMessage=It's powered by the connection we have with our fans and with each other.}
  }
 }
 Speech_C_1 {
  id=#14
  class=Speech_C
  outputs [#9]
  properties {nodeName="", softTaskIcon=&T_UI_NPC_question, generalConfig {defaultMessage="Happy fans equals happy Honmoon!"}}
 }
]
